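# Physics asset for the Mascot Militia Tomato backpack cosmetic, defining bodies and joint limits for its leaf and rope bones.
# Five box-shaped leaves constrain directly to the capsule pack body, with the first leaf carrying an extra end segment.
# A two-link capsule rope chain hangs off the pack with wider swing limits than the leaves.
# All joints have tight twist limits, and the third leaf is the stiffest in swing range.
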PhysicsAsset /BRCosmetics/Accessories/FORT_Backpacks/M_MED_MascotMilitiaTomato_Backpack/Meshes/M_MED_Mascot_Militia_Tomato_Pack_Physics
bodies {
 dyn_leaf_01 [{shapeType=Box, extentX=8, extentY=4, extentZ=2, center=<2 0 0>}]
 dyn_leaf_01_end [{shapeType=Box, extentX=5, extentY=2.5, extentZ=2, center=<3 0 0>}]
 dyn_leaf_02 [{shapeType=Box, extentX=8, extentY=4, extentZ=2, center=<2 0 0>}]
 dyn_leaf_03 [{shapeType=Box, extentX=8, extentY=4, extentZ=2, center=<2 0 0>}]
 dyn_leaf_04 [{shapeType=Box, extentX=8, extentY=4, extentZ=2, center=<2 0 0>}]
 dyn_leaf_05 [{shapeType=Box, extentX=8, extentY=4, extentZ=2, center=<2 0 0>}]
 dyn_rope_01 [{shapeType=Capsule, length=2, radius=1.5, center=<3.25 0 0>, rotation=<rot 90 0 0>}]
 dyn_rope_02 [{shapeType=Capsule, length=2, radius=1.5, center=<3.25 0 0>, rotation=<rot 90 0 0>}]
 pack [{shapeType=Capsule, length=2, radius=2}]
}
constraints [
 {bone1Name=dyn_rope_01, bone2Name=pack, swing1LimitDegrees=60, swing2LimitDegrees=60, twistLimitDegrees=5}
 {bone1Name=dyn_rope_02, bone2Name=dyn_rope_01, swing1LimitDegrees=60, swing2LimitDegrees=60, twistLimitDegrees=5}
 {bone1Name=dyn_leaf_05, bone2Name=pack, swing1LimitDegrees=20, swing2LimitDegrees=30, twistLimitDegrees=2}
 {bone1Name=dyn_leaf_04, bone2Name=pack, swing1LimitDegrees=20, swing2LimitDegrees=30, twistLimitDegrees=2}
 {bone1Name=dyn_leaf_03, bone2Name=pack, swing1LimitDegrees=10, swing2LimitDegrees=15, twistLimitDegrees=2}
 {bone1Name=dyn_leaf_02, bone2Name=pack, swing1LimitDegrees=20, swing2LimitDegrees=30, twistLimitDegrees=2}
 {bone1Name=dyn_leaf_01, bone2Name=pack, swing1LimitDegrees=20, swing2LimitDegrees=30, twistLimitDegrees=2}
 {bone1Name=dyn_leaf_01_end, bone2Name=dyn_leaf_01, swing1LimitDegrees=20, swing2LimitDegrees=30, twistLimitDegrees=2}
]
