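# Animation Blueprint for the Mascot Militia Tomato backpack that adds physics motion to the cosmetic.
# The AnimGraph passes the input pose through a RigidBody node using the pack's physics asset, overridden gravity, and scaled, clamped component-space inertia.
# Each animation update checks bIsTargeting and sets PhyAlpha, the RigidBody blend weight, lower while targeting and to full otherwise, damping the simulated motion during aiming.
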AnimBlueprint /BRCosmetics/Accessories/FORT_Backpacks/M_MED_MascotMilitiaTomato_Backpack/Meshes/M_MED_Mascot_Militia_Tomato_Pack_AnimBlueprint
target_skeleton=&/M_MED_Mascot_Militia_Tomato_Pack_Skeleton
graphs [
 AnimGraph {nodes [
   AnimGraphNode_SubInput_0 = Input Pose()
   n1 = RigidBody(ComponentPose=Local To Component(LocalPose=@AnimGraphNode_SubInput_0), Alpha=Get PhyAlpha() -> (PhyAlpha:double)) {
    bClampLinearTranslationLimitToRefPose=true
    bOverrideWorldGravity=true
    overridePhysicsAsset=&/M_MED_Mascot_Militia_Tomato_Pack_Physics
    alphaBoolBlend {blendOption=Linear}
    componentAppliedLinearAccClamp=<2000 2000 2000>
    componentLinearAccScale=<0.4 0.4 0.3>
    componentLinearVelScale=<0.4 0.4 0.3>
    overrideWorldGravity {z=-980}
   }
   Output Pose(Result=Component To Local(ComponentPose=@n1)) {layerGroup=None}
  ]}
 EventGraph {nodes [
   Event BlueprintUpdateAnimation()
   > n1 = Branch(Condition=Get bIsTargeting())
   > Set PhyAlpha(PhyAlpha:double=0.3) -> (Output_Get:double)
   Set PhyAlpha(execute=@n1.else, PhyAlpha=1)
   TryGetPawnOwner()
  ]}
]
variables [PhyAlpha {type=number, default=1}]
